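; Timed blizzard dome actor for the Blizzard Grenade that applies an in-dome GameplayEffect to pawns inside, taking team from its instigator.
; Hotfixable values set sphere radius and lifespan.
; The server adds a looping GameplayCue, catches initial overlaps, then tracks enter/exit to apply or remove the effect, clearing both cue and tracked effects on death.
; Clients spawn a quality-gated wash force physics field, reversing its fade timeline before destroying it.
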(event EventBeginPlay
  (bind _blizzardareavolume (Variables|Default|GetBlizzardAreaVolume))
  (Utilities|FlowControl|Sequence
    (:then_0
      ;[1 1 1] Set team based on instigator for proper affiliation filtering
      (bind _asfort_pawn (Utilities|Casting|CastToFortPawn (Game|GetInstigator)))
      (CallFunction|SetTeam (Class|FortPawn|GetTeam _asfort_pawn)))
    (:then_1
      ;[1 1 1] Update radius based on hotfix value
      (Components|Sphere|SetSphereRadius (Variables|Default|GetBlizzardAreaVolume) (ScalableFloat|GetValueatLevel (Variables|Default|GetHF_DomeRadius)) true)
      (Actor|SetLifeSpan (ScalableFloat|GetValueatLevel (Variables|Default|GetBlizzardDuration)))
      (Networking|SwitchHasAuthority
        (:Authority
          (Utilities|FlowControl|Sequence
            (:then_0
              ;[0 0.09009 0.15] Add FX (removed on death)
              (GameplayCue|AddGameplayCueOnActor(Looping) self (Variables|Default|GetGCTag_Dome) (Ability|GameplayCue|MakeGameplayCueParameters :NormalizedMagnitude 0.0 :RawMagnitude (ScalableFloat|GetValueatLevel (Variables|Default|GetHF_DomeRadius)) :Location "0,0,0" :Normal "0,0,0" :GameplayEffectLevel 1 :AbilityLevel 1 :TargetAttachComponent (Variables|Transformation|GetRootComponent) :bReplicateLocationWhenUsingMinimalRepProxy false)))
            (:then_1
              ;[0 0.09009 0.15] Check for pawns that are already overlapping on spawn
              (for _array_element (Collision|GetOverlappingActors (Variables|Default|GetBlizzardAreaVolume) "/Script/FortniteGame.FortPawn" :!class /Script/Engine.PrimitiveComponent)
                (CallFunction|EnteringBlizzardAoE 0 _array_element)))
            (:then_2
              ;[0 0.09009 0.15] Listen for overlap enter/exit events for application/removal of GEs
              (Collision|BindEventtoOnComponentBeginOverlap _blizzardareavolume (delegate Custom|EnteringBlizzardAoE))
              (Collision|BindEventtoOnComponentEndOverlap _blizzardareavolume (delegate Custom|ExitingBlizzardAoE)))))
        (:Remote
          (Utilities|EffectsQualityCheck 0
            (:Cinematic
              _)
            (:Epic
              _))
          (bind _returnvalue (Game|SpawnActorfromClass "/Game/Effects/Niagara/Blueprints/PhsyicsFields/FN_WashForce.FN_WashForce_C" (Transformation|GetActorTransform) "Undefined" "MultiplyWithRoot" 0 9.5e+02 1.5 -999.0 0))
          (Variables|Default|SetWash%20Force _returnvalue)
          (Utilities|FlowControl|Delay (+ (ScalableFloat|GetValueatLevel (Variables|Default|GetBlizzardDuration)) 0.5))
          (Components|Timeline|ReversefromEnd (Variables|BGA_BigBlizzardAoE|GetForceFadeTimeline))
          (Utilities|FlowControl|Delay 0.6)
          (bind _wash_force (Variables|Default|GetWash%20Force)
            (:then
              (Actor|DestroyActor _wash_force))
            (:else)))))))

(event Custom|EnteringBlizzardAoE
  (params (OverlappedComponent :type Primitive%20Component%20Object%20Reference) (OtherActor :type Actor%20Object%20Reference) (OtherComp :type Primitive%20Component%20Object%20Reference) (OtherBodyIndex :type Integer) (bFromSweep :type Boolean) (SweepResult :type Hit%20Result%20Structure%20%28by%20ref%29))
  ;[0 0.09009 0.15] Listen for overlap enter/exit events for application/removal of GEs
  (Utilities|IsValid (Ability|GetAbilitySystemComponent OtherActor)
    (:Is%20Valid
      (Utilities|Array|AddUnique (Variables|Default|GetAffectedActors) OtherActor)
      (GameplayEffects|ApplyGameplayEffectToTarget (Ability|GetAbilitySystemComponent self) (Variables|Default|GetGE_Blizzard_InDome) (Ability|GetAbilitySystemComponent OtherActor) 0.0 (Class|AthenaFunctionLibrary|MakeSimpleFortEffectContext (Game|GetInstigator) self)))))

(event Custom|ExitingBlizzardAoE
  (params (OverlappedComponent :type Primitive%20Component%20Object%20Reference) (OtherActor :type Actor%20Object%20Reference) (OtherComp :type Primitive%20Component%20Object%20Reference) (OtherBodyIndex :type Integer))
  (bind _returnvalue (Ability|GetAbilitySystemComponent OtherActor))
  ;[0 0.09009 0.15] Listen for overlap enter/exit events for application/removal of GEs
  (Utilities|IsValid _returnvalue
    (:Is%20Valid
      (GameplayEffects|RemoveActiveGameplayEffectbySourceEffect _returnvalue (Variables|Default|GetGE_Blizzard_InDome) 0 1)
      (Utilities|Array|RemoveItem (Variables|Default|GetAffectedActors) OtherActor))))

(event EventEndPlay
  (params (EndPlayReason :type EEndPlayReason%20Enum :default "Destroyed"))
  (|CallParentFunction EndPlayReason)
  (Actor|DestroyActor (Variables|Default|GetWash%20Force)))

(event EventOnDeathServer
  (params (Damage :type Float%20%28single-precision%29) (DamageTags :type Gameplay%20Tag%20Container%20Structure%20%28by%20ref%29) (Momentum :type Vector :default "0, 0, 0") (HitInfo :type Hit%20Result%20Structure%20%28by%20ref%29) (InstigatedBy :type Controller%20Object%20Reference) (DamageCauser :type Actor%20Object%20Reference) (EffectContext :type Gameplay%20Effect%20Context%20Handle%20Structure))
  ;[0 0.09009 0.15] Remove GameplayCue on death
  (GameplayCue|RemoveGameplayCueOnActor(Looping) self (Variables|Default|GetGCTag_Dome) (Ability|GameplayCue|MakeGameplayCueParameters 0.0))
  ;[0 0.09009 0.15] Failsafe: Remove GE from any actor that didn't receive the exit event yet
  (for _array_element (Variables|Default|GetAffectedActors)
    (CallFunction|ExitingBlizzardAoE 0 _array_element)))

(fn ConstructionScript)
