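; Fire Petal Blizzard Grenade ability that gives its avatar damage and slow effects while inside blizzard dome volumes.
; Activation applies each enabled effect from the granting instigator, using targeting-preset filtering that an always-target-instigator setting can bypass for the instigator.
; The server ends the ability when the InDome tag is removed, and ending removes effects by asset tag.
; The tag stacks across overlapping volumes instead of the effects, working around stack magnitude limitations.
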(event Ability|EventOnEndAbility
  (params (bWasCancelled :type Boolean))
  (CallFunction|RemoveEffects))

(event Ability|EventActivateAbility
  ; To work around stack magnitude calculation limitations, this GA handles applying and removing the GEs once when the player enters any volume, and once when leaving any volume (i.e. the InDome granted tag stacks, but not the GEs themselves)
  (CallFunction|ApplyEffects)
  (Ability|HasAuthority
    (:True
      (Ability|Tasks|WaitGameplayTagRemove (Variables|Default|GetInDomeGrantedTag)
        (:Removed
          (Ability|EndAbility))))))

(event Custom|ApplyEffects
  (Variables|Default|SetAvatarArray (Utilities|Array|MakeArray (Ability|GetAvatarActorfromActorInfo)))
  (bind _output_get (Variables|Default|SetGrantedByContext (Ability|GetGrantedbyEffectContext)))
  (Variables|Default|SetInstigatorToUseForFiltering (Ability|EffectContext|GetInstigatorActor _output_get))
  (if (ScalableFloat|GetValueasBool (Variables|Default|GetRow_ShouldApplyDamage))
    (CallFunction|FilterandApplyEffect (Variables|Default|GetTargetingPreset_Damage) (Variables|Default|GetGE_Blizzard_Damage))
    (else
      _))
  (if (ScalableFloat|GetValueasBool (Variables|Default|GetRow_ShouldApplySlow))
    (CallFunction|FilterandApplyEffect (Variables|Default|GetTargetingPreset_Slow) (Variables|Default|GetGE_Blizzard_Slow))))

(event Custom|FilterAndApplyEffect
  (params (TargetingPresetToFilterWith :type Targeting%20Preset%20Object%20Reference) (GE_ToApply :type Gameplay%20Effect%20Class%20Reference))
  (if (not (and (ScalableFloat|GetValueasBool (Variables|Default|GetHF_AlwaysTargetInstigator)) (CallFunction|IsTargetingInstigator)))
    (bind (_asynctaskproxy _targetinghandle) (Class|AsyncActionPerformTargeting|PerformFilteringAsyncAction (Variables|Default|GetInstigatorToUseForFiltering) TargetingPresetToFilterWith false (Variables|Default|GetAvatarArray))
      (:Targeted
        (if (Utilities|Array|IsNotEmpty (TargetingSystem|TargetingResults|GetTargetingResultsActors (GameInstanceSubsystems|GetTargetingSubsystem) _targetinghandle))
          _))))
  (GameplayEffects|ApplyGameplayEffectToTarget (Ability|GetAbilitySystemComponent (Variables|Default|GetInstigatorToUseForFiltering)) GE_ToApply (Ability|GetAbilitySystemComponentfromActorInfo) 0.0 (Variables|Default|GetGrantedByContext)))

(event Custom|RemoveEffects
  (Ability|RemoveGameplayEffectFromOwnerWithAssetTags (Variables|Default|GetAssetTagsToRemove)))

(fn IsTargetingInstigator
  (return (== (Ability|GetAbilitySystemComponentfromActorInfo) (Ability|GetAbilitySystemComponent (Variables|Default|GetInstigatorToUseForFiltering)))))
